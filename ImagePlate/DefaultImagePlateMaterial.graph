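# Default material for the ImagePlate plugin, an unlit, two-sided surface for showing an image on a plate.
# A single texture sample reads a texture object parameter named InputTexture, so material instances can swap the displayed image.
# The parameter falls back to the engine's DefaultDiffuse texture.
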Material /ImagePlate/DefaultImagePlateMaterial
graph {
 properties {pixelDepthOffsetMode=PDOM_Legacy, shadingModel=MSM_Unlit, twoSided=true, usedShadingModels=Unlit}
 nodes [TextureSample(Tex=TextureObjectParameter() {parameterName=InputTexture, texture=/Engine/EngineMaterials/DefaultDiffuse})]
}
